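# Sound cue for outdoor insect ambience in the Grassland_01 biome, playing insect-by one-shot variations.
# A quality-level node sends the two lower levels to a random pick of five base waves and the two higher levels to a random pick of their five stereo counterparts, all equally weighted.
# It uses a non-spatialized tile-ambient one-shot attenuation and the outdoor ambient sound class at roughly half volume.
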SoundCue /Game/Sounds/Apollo/Biomes/Grassland_01/Biome_Grassland_01_Insects_Cue
root=QualityLevel_0
nodes [
 {id=QualityLevel_0, children [Random_1,Random_1,Random_0,Random_0]}
 {id=Random_0, children [WavePlayer_1,WavePlayer_2,WavePlayer_4,WavePlayer_5,WavePlayer_6], properties {weights [1,1,1,1,1]}}
 {id=Random_1, children [WavePlayer_0, WavePlayer_7, WavePlayer_8, WavePlayer_10, WavePlayer_11], properties {weights [1,1,1,1,1]}}
 {id=WavePlayer_0, properties {soundWaveAssetPtr=&/Biome_Grassland_01_InsectBy_04}}
 {id=WavePlayer_1, properties {soundWaveAssetPtr=&/Biome_Grassland_01_InsectBy_01_Stereo}}
 {id=WavePlayer_10, properties {soundWaveAssetPtr=&/Biome_Grassland_01_InsectBy_05}}
 {id=WavePlayer_11, properties {soundWaveAssetPtr=&/Biome_Grassland_01_InsectBy_06}}
 {id=WavePlayer_2, properties {soundWaveAssetPtr=&/Biome_Grassland_01_InsectBy_02_Stereo}}
 {id=WavePlayer_4, properties {soundWaveAssetPtr=&/Biome_Grassland_01_InsectBy_04_Stereo}}
 {id=WavePlayer_5, properties {soundWaveAssetPtr=&/Biome_Grassland_01_InsectBy_05_Stereo}}
 {id=WavePlayer_6, properties {soundWaveAssetPtr=&/Biome_Grassland_01_InsectBy_06_Stereo}}
 {id=WavePlayer_7, properties {soundWaveAssetPtr=&/Biome_Grassland_01_InsectBy_01}}
 {id=WavePlayer_8, properties {soundWaveAssetPtr=&/Biome_Grassland_01_InsectBy_02}}
]
properties {
 attenuationSettings=/Game/Sounds/Attenuation/Fort_TileAmbient_OneShotNoSpatial_Att
 maxDistance=8000
 soundClassObject=&Fort_Ambient_Outdoors_Athena
 volumeMultiplier=0.501
 preEffectBusSends [{}]
}
